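# Idle Niagara effect for the BucketKick pickaxe, adding purple mesh flames and a pulsing glow at sockets read through a socket reader.
# Flames_moving continuously spawns two spawn groups that travel along curve-driven socket offsets, grow and shrink over life, and feed scale, time and random values to dynamic material parameters.
# Glow burst-spawns a persistent sprite with an attached light, its color modulated by a waveform.
# A static six-flame socket emitter is present but disabled.
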NiagaraSystem /BRCosmetics/Weapons/FORT_Melee/Pickaxe_BucketKick/FX/NS_Pickaxe_BucketKick_Idle
dependencies {
 usedDataInterfaces [
  &NiagaraDataInterfaceArrayFloat
  &NiagaraDataInterfaceArrayFloat3
  &NiagaraDataInterfaceVectorCurve
  &NiagaraDataInterfaceCurve
  &NiagaraDataInterfaceSocketReader
 ]
 usedDynamicInputs []
 usedModules [
  &self:SystemSpawnScript.SystemScriptSource.SystemScriptGraph.NiagaraNodeAssignment_0.SetVariables_#1
  &SystemState
 ]
}
emitters {
 Flames_Socket {
  bEnabled=false
  emitterUpdateScript [
   &EmitterState(Life Cycle Mode=::NewEnumerator1, Loop Behavior=::NewEnumerator1, Loop Duration Mode=::NewEnumerator1, UseLoopDelay=false)
   &SpawnBurst_Instantaneous(Spawn Count=6, Use Spawn Probability=false, UseLoopCountLimit=false)
  ]
  particleSpawnScript [
   &InitializeParticle(
    Color=<rgba 1 0.2795 0.653 1>
    Mesh Scale=&SelectVectorFromArray(
     Direct Array Index=&ReturnExecIndex()
     Vector Selection Array=DataInterface() {propertyValues {FloatData [<0.9 0.9 0.9>, <0 0 0>, <0.7 0.7 0.7>, <1 1 1>, <0.7 0.7 0.7>, <0.7 0.7 0.7>, <1 1 1>]}}
    )
    Mesh Scale Mode=::NewEnumerator4
    UsePositionOffset=false
    Write Material Random=true
   )
   &InitialMeshOrientation(
    Mesh Axis To Orient=<0 0 1>
    Mesh Orientation Mode=::NewEnumerator2
    Orientation Vector=&SelectVectorFromArray(
     Direct Array Index=&ReturnExecIndex()
     Vector Selection Array=DataInterface() {propertyValues {FloatData [<0 0 0.5>, <0 0.1 1>, <-0.3 0 1>, <0 0.3 -1>, <0 -0.4 -1>, <1 0 1>, <1 0 0>]}}
    )
    Use Rotation Vector=false
   )
   /SetVariables(
    Particles.R2_Color:LinearColor=<rgba 1 0 0.4853 1>
    Particles.R2_Scale:Vector3f=&Multiply_Vector(A=$Particles.Scale, B=<1.2 1.2 1.2>)
   ) {isSetParameters=true, module=SetVariables_#2}
   /SetVariables(Particles.RandU:NiagaraFloat=&RandomRangeFloat(Minimum=0, Maximum=1)) {isSetParameters=true, module=SetVariables_#3}
  ]
  particleUpdateScript [
   &ParticleState(Kill Particles When Lifetime Has Elapsed=false, Let Infinitely Lived Particles Die When Emitter Deactivates=true)
   &SocketLocation(
    SamplePosOffset=true
    Sampled Position Offset=&SelectVectorFromArray(
     Direct Array Index=&ReturnExecIndex()
     Vector Selection Array=DataInterface() {propertyValues {FloatData [<8 0 12>, <9.5 0 20>, <9 0 28.5>, <6 0 -10>, <6 0 -23.5>, <-6 0 7>, <0 0 0>]}}
    )
    Socket Reader=$System.Socket Reader Interface
    Socket Sampling Mode=::NewEnumerator3
    VelocitySpeedLimit=false
   )
   &DynamicMaterialParameters(
    Index 0 Param 1=&Multiply_Float(
     A=1
     B=&SelectFloatFromArray(
      Direct Array Index=&ReturnExecIndex()
      Float Selection Array=DataInterface() {propertyValues {FloatData [1,1,1,1,1,1,1]}}
     )
    )
    Index 0 Param 2=&Multiply_Float(A=$Engine.Time, B=1)
    Index 0 Param 3=&SelectFloatFromArray(
     Direct Array Index=&ReturnExecIndex()
     Float Selection Array=DataInterface() {propertyValues {FloatData [1,1,1,1,1,1,1]}}
    )
    Index 0 Param 4=$Particles.RandU
    Module.Index0Param0Write=true
    Module.Index0Param1Write=true
    Module.Index0Param2Write=true
    Module.Index0Param3Write=true
    Module.Index2Param0Write=true
    Module.Index2Param1Write=true
    Module.Index2Param2Write=true
    Module.Index2Param3Write=true
    Write Parameter Index 2=true
   )
   /SetVariables(Particles.R2_DynMatParm:Vector4f=&MakeVector4(
     X=&Multiply_Float(
      A=1.15
      B=&SelectFloatFromArray(
       Direct Array Index=&ReturnExecIndex()
       Float Selection Array=DataInterface() {propertyValues {FloatData [1,1,1,1,1,1,1]}}
      )
     )
     Y=&Multiply_Float(A=$Engine.Time, B=1)
     Z=&SelectFloatFromArray(
      Direct Array Index=&ReturnExecIndex()
      Float Selection Array=DataInterface() {propertyValues {FloatData [1,1,1,1,1,1,1]}}
     )
     W=$Particles.RandU
    )) {isSetParameters=true, module=SetVariables_#4}
   /SetVariables() {isSetParameters=true, module=SetVariables_#5}
  ]
  rendererClasses [&NiagaraMeshRendererProperties]
  renderers [
   {rendererClass=&NiagaraMeshRendererProperties, rendererIndex=0}
   {rendererClass=&NiagaraMeshRendererProperties, rendererIndex=1}
  ]
 }
 Flames_moving {
  emitterUpdateScript [
   &EmitterState(Life Cycle Mode=::NewEnumerator1, Loop Behavior=::NewEnumerator1, Loop Duration Mode=::NewEnumerator1, UseLoopDelay=false)
   &SpawnRate(SpawnRate=2, Use Spawn Probability=false)
   &SpawnBurst_Instantaneous(Spawn Count=6, Use Spawn Probability=false, UseLoopCountLimit=false) {enabled=false}
  ]
  particleSpawnScript [
   &FN_SetSpawnGroup(Number of Spawn Groups=2)
   /SetVariables(
    Particles.RandSize:NiagaraFloat=&RandomRangeFloat(Minimum=0.8, Maximum=1.2)
    Particles.RandU:NiagaraFloat=&RandomRangeFloat(Minimum=0, Maximum=1)
   ) {isSetParameters=true, module=SetVariables_#3}
   &InitializeParticle(
    Color=<rgba 0.2771 0 0.4896 1>
    Lifetime=2
    Mesh Scale=&MaskVectorBySpawnGroup(Masked Vector=<1.5 1.5 1.5>, Passthrough Vector=<1.5 1.5 1.5>)
    Mesh Scale Mode=::NewEnumerator4
    UsePositionOffset=false
    Write Material Random=true
   )
   &InitialMeshOrientation(
    Mesh Axis To Orient=&MaskVectorBySpawnGroup(Masked Vector=<1 0 0.5>, Passthrough Vector=<0 -1 -0.5>)
    Mesh Orientation Mode=::NewEnumerator2
    Orientation Vector=&MaskVectorBySpawnGroup(Masked Vector=<1 0 0>, Passthrough Vector=<0 1 0>)
    Use Rotation Vector=false
   )
   /SetVariables(
    Particles.R2_Color:LinearColor=<rgba 1 0 0.4853 1>
    Particles.R2_Scale:Vector3f=&Multiply_Vector(A=$Particles.Scale, B=<1.2 1.2 1.2>)
   ) {isSetParameters=true, module=SetVariables_#2}
  ]
  particleUpdateScript [
   &ParticleState()
   &SocketLocation(
    Orientation Sampling=::NewEnumerator1
    SamplePosOffset=true
    Sampled Position Offset=&MaskVectorBySpawnGroup(
     Masked Vector=&VectorFromCurve(VectorCurve=DataInterface() {propertyValues {
        XCurve {keys [
          {arriveTangent=16.5157, interpMode=RCIM_Cubic, leaveTangent=16.5157, tangentMode=RCTM_User, time=0, value=6.25}
          {arriveTangent=6.2989, interpMode=RCIM_Cubic, leaveTangent=6.2989, tangentMode=RCTM_User, time=0.5, value=10}
          {arriveTangent=-16, interpMode=RCIM_Cubic, leaveTangent=-16, time=0.8, value=10}
          {arriveTangent=-10.9981, interpMode=RCIM_Cubic, leaveTangent=-10.9981, tangentMode=RCTM_User, time=1, value=2}
         ]}
        YCurve {keys [<at 0 0>,<at 1 0>]}
        ZCurve {keys [<at 0 12>,<at 1 31.25>]}
       }})
     Passthrough Vector=&VectorFromCurve(VectorCurve=DataInterface() {propertyValues {XCurve {keys [<at 0 6>,<at 1 6>]}, YCurve {keys [<at 0 0>,<at 1 0>]}, ZCurve {keys [<at 0 -9.25>,<at 1 -23>]}}})
    )
    Socket Reader=$System.Socket Reader Interface
    Socket Sampling Mode=::NewEnumerator3
    VelocitySpeedLimit=false
   )
   &SkeletalMeshLocation(
    MaskSpawnGroup=false
    Mesh Sampling Type=::NewEnumerator5
    Position Sampling=::NewEnumerator1
    Sampled Position Offset=&VectorFromCurve(VectorCurve=DataInterface() {propertyValues {
       XCurve {keys [
         {arriveTangent=7.3708, interpMode=RCIM_Cubic, leaveTangent=7.3708, tangentMode=RCTM_User, time=0, value=5}
         {arriveTangent=2, interpMode=RCIM_Cubic, leaveTangent=2, time=0.5, value=8}
         {arriveTangent=-3.0898, interpMode=RCIM_Cubic, leaveTangent=-3.0898, tangentMode=RCTM_User, time=1, value=7}
        ]}
       YCurve {keys [<at 0 0>,<at 1 0>]}
       ZCurve {keys [<at 0 12>,<at 1 25>]}
      }})
    Skeletal Mesh=$System.Skeletal Mesh
    Socket Sampling Mode=::NewEnumerator5
   ) {enabled=false}
   &UpdateMeshOrientation(
    Rotation Rate=&MaskFloatBySpawnGroup(Masked Float=0.12, Passthrough Float=0)
    Rotation Vector=&MaskVectorBySpawnGroup(Masked Vector=<1 0 0>, Passthrough Vector=<1 0 0>)
   )
   /SetVariables() {enabled=false, isSetParameters=true}
   &ScaleMeshSize(Scale Factor=&VectorFromFloat(Value=&FloatFromCurve(FloatCurve=DataInterface() {propertyValues {Curve {keys [
          {arriveTangent=5.3398, interpMode=RCIM_Cubic, leaveTangent=5.3398, tangentMode=RCTM_User, time=0, value=0}
          {interpMode=RCIM_Cubic, tangentMode=RCTM_User, time=0.5, value=1}
          {arriveTangent=-2.4563, tangentMode=RCTM_User, time=1, value=0}
         ]}}})))
   &DynamicMaterialParameters(
    Index 0 Param 1=&Multiply_Float(A=&MakeFloatFromVector(Vector=$Particles.Scale, Channel=::NewEnumerator0), B=1.5)
    Index 0 Param 2=&Multiply_Float(A=$Engine.Time, B=1)
    Index 0 Param 3=1
    Index 0 Param 4=$Particles.RandU
    Module.Index0Param0Write=true
    Module.Index0Param1Write=true
    Module.Index0Param2Write=true
    Module.Index0Param3Write=true
    Module.Index2Param0Write=true
    Module.Index2Param1Write=true
    Module.Index2Param2Write=true
    Module.Index2Param3Write=true
    Write Parameter Index 2=true
   )
   /SetVariables(Particles.R2_DynMatParm:Vector4f=&MakeVector4(
     X=&Multiply_Float(A=&MakeFloatFromVector(Vector=$Particles.Scale, Channel=::NewEnumerator0), B=1.7)
     Y=&Multiply_Float(A=$Engine.Time, B=1)
     Z=1
     W=$Particles.RandU
    )) {isSetParameters=true, module=SetVariables_#4}
   /SetVariables() {isSetParameters=true, module=SetVariables_#5}
   &ScaleColor(Scale RGB=&MakeVectorFromLinearColorRGB(Color=<rgba 0.2448 0 0.4896 1>), ScaleA=true, ScaleRGB=true)
  ]
  rendererClasses [&NiagaraMeshRendererProperties]
  renderers [{rendererClass=&NiagaraMeshRendererProperties, rendererIndex=0}]
 }
 Glow {
  emitterUpdateScript [
   &EmitterState(
    Inactive Response=::NewEnumerator1
    Life Cycle Mode=::NewEnumerator1
    Loop Behavior=::NewEnumerator1
    Loop Duration Mode=::NewEnumerator1
    UseLoopDelay=false
   )
   &SpawnBurst_Instantaneous(Use Spawn Probability=false, UseLoopCountLimit=false)
  ]
  particleSpawnScript [
   &InitializeParticle(
    Color=<rgba 0.5561 0 1 0.5>
    Lifetime=5
    Sprite Size Mode=::NewEnumerator3
    Uniform Sprite Size=50
    UsePositionOffset=false
    Write Material Random=true
   )
   &SocketLocation(SamplePosOffset=true, Socket Reader=$System.Socket Reader Interface, Socket Sampling Mode=::NewEnumerator3, VelocitySpeedLimit=false)
  ]
  particleUpdateScript [
   &ParticleState(Kill Particles When Lifetime Has Elapsed=false, Let Infinitely Lived Particles Die When Emitter Deactivates=true)
   &FN_Light_Attributes(
    Color=$Particles.Color
    Create Color Attribute=true
    Create Exponent Attribute=true
    Create Position Attribute=true
    Create Radius Attribute=true
    Enable Particle Light Enable=true
    Exponent=5
    Position=&AddVectorToPosition(Position=$Particles.Position, Vector=<0 -20 0>)
    Radius=65
    Scale Color=8
   )
   &ScaleColor(
    Scale RGB=&VectorFromFloat(Value=&Waveform(Global Amplitude Scale=0.5, "Waveform [1]"=::NewEnumerator12, "[1] Frequency"=10, "[1] Amplitude Min/Max"=<0 1>, "[1] Bias"=1))
    ScaleA=true
    ScaleRGB=true
   )
  ]
  rendererClasses [&NiagaraSpriteRendererProperties, &NiagaraLightRendererProperties]
  renderers [
   {rendererClass=&NiagaraSpriteRendererProperties}
   {rendererClass=&NiagaraLightRendererProperties, rendererIndex=1}
  ]
 }
}
summary {
 systemName=NS_Pickaxe_BucketKick_Idle
 emitters [
  {bEnabled=false, emitterName=Flames_Socket, rendererClasses [&NiagaraMeshRendererProperties]}
  {emitterName=Flames_moving, rendererClasses [&NiagaraMeshRendererProperties]}
  {emitterName=Glow, rendererClasses [&NiagaraSpriteRendererProperties, &NiagaraLightRendererProperties]}
 ]
 userVariables [
  User.bIsFrontEnd {type=NiagaraBool}
  User.bIsFrontEndPreview {type=NiagaraBool}
 ]
}
