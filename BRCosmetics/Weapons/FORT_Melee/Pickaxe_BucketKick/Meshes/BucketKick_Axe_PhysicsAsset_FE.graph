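# Physics asset for the BucketKick two-handed pickaxe cosmetic, providing collision bodies and joint limits for its dangling chain and cloth bones.
# Three constrained capsule strands, a four-link chain and three-link cloth A and B, each attach to a small master sphere.
# Chain and cloth A lock twist with limited swing, while cloth B allows slight twist.
# The handle capsule and large cloth C capsule have no constraints, nor do the spheres link to the handle.
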PhysicsAsset /BRCosmetics/Weapons/FORT_Melee/Pickaxe_BucketKick/Meshes/BucketKick_Axe_PhysicsAsset_FE
bodies {
 Melee_TwoHanded_Handle [{shapeType=Capsule, length=117.7707, radius=2.0374, center=<0.0556 -1.1674 32.2738>, rotation=<rot 0.0172 -89.9995 0.0257>}]
 dyn_chain_01 [{shapeType=Capsule, length=2.5, radius=1, shapeName=dyn_chain_09_capsule, center=<2.3373 0 0>, rotation=<rot -90 90 -89.9999>}]
 dyn_chain_02 [{shapeType=Capsule, length=2.5, radius=1, shapeName=dyn_chain_09_capsule, center=<2.3373 0 0>, rotation=<rot -90 90 -89.9999>}]
 dyn_chain_03 [{shapeType=Capsule, length=2.5, radius=1, shapeName=dyn_chain_09_capsule, center=<2.3373 0 0>, rotation=<rot -90 90 -89.9999>}]
 dyn_chain_04 [{shapeType=Capsule, length=2.5, radius=1, shapeName=dyn_chain_09_capsule, center=<2.3373 0 0>, rotation=<rot -90 90 -89.9999>}]
 dyn_cloth_A_01 [{shapeType=Capsule, length=2.5, radius=3, shapeName=dyn_chain_09_capsule, center=<2.3374 -0.8671 1.2952>, rotation=<rot 0 -179.9996 123.8002>}]
 dyn_cloth_A_02 [{shapeType=Capsule, length=9, radius=2.5, shapeName=dyn_chain_09_capsule, center=<2.3374 0 0>, rotation=<rot -0.0001 -179.9998 -47.7785>}]
 dyn_cloth_A_03 [{shapeType=Capsule, length=8, radius=3.5, shapeName=dyn_chain_09_capsule, center=<2.3374 0 0>, rotation=<rot 0 -179.9999 -42.7678>}]
 dyn_cloth_B_01 [{shapeType=Capsule, length=2.5, radius=3, shapeName=dyn_chain_09_capsule, center=<2.3374 -0.8671 1.2952>, rotation=<rot 0 -179.9996 123.8002>}]
 dyn_cloth_B_02 [{shapeType=Capsule, length=8, radius=3.5, shapeName=dyn_chain_09_capsule, center=<2.3374 0 0>, rotation=<rot 0 -180 0>}]
 dyn_cloth_B_03 [{shapeType=Capsule, length=7, radius=3.5, shapeName=dyn_chain_09_capsule, center=<2.3374 0 0>, rotation=<rot 0 -180 0>}]
 dyn_cloth_C_01 [{shapeType=Capsule, length=27, radius=21, shapeName=dyn_cloth_C_01_capsule, center=<0 -6.7238 19.3567>, rotation=<rot -90 0 0>}]
 mast_cloth_A [{shapeType=Sphere, radius=0.505, shapeName=mast_chain_sphere}]
 mast_cloth_B [{shapeType=Sphere, radius=0.505, shapeName=mast_chain_sphere}]
 mstr_chain [{shapeType=Sphere, radius=0.505, shapeName=mast_chain_sphere}]
}
constraints [
 {bone1Name=dyn_cloth_B_03, bone2Name=dyn_cloth_B_02, swing1LimitDegrees=15, swing2LimitDegrees=20, twistLimitDegrees=2}
 {bone1Name=dyn_cloth_B_02, bone2Name=dyn_cloth_B_01, swing1LimitDegrees=15, swing2LimitDegrees=20, twistLimitDegrees=2}
 {bone1Name=mast_cloth_B, bone2Name=dyn_cloth_B_01, swing1LimitDegrees=15, swing2LimitDegrees=20, twistLimitDegrees=2}
 {bone1Name=dyn_chain_04, bone2Name=dyn_chain_03, swing1LimitDegrees=30, swing2LimitDegrees=30, twistLimitDegrees=45, twistMotion=Locked}
 {bone1Name=dyn_chain_03, bone2Name=dyn_chain_02, swing1LimitDegrees=30, swing2LimitDegrees=30, twistLimitDegrees=45, twistMotion=Locked}
 {bone1Name=dyn_chain_02, bone2Name=dyn_chain_01, swing1LimitDegrees=30, swing2LimitDegrees=30, twistLimitDegrees=45, twistMotion=Locked}
 {bone1Name=mstr_chain, bone2Name=dyn_chain_01, swing1LimitDegrees=30, swing2LimitDegrees=30, twistLimitDegrees=45, twistMotion=Locked}
 {bone1Name=dyn_cloth_A_03, bone2Name=dyn_cloth_A_02, swing1LimitDegrees=10, swing2LimitDegrees=10, twistLimitDegrees=45, twistMotion=Locked}
 {bone1Name=dyn_cloth_A_02, bone2Name=dyn_cloth_A_01, swing1LimitDegrees=10, swing2LimitDegrees=10, twistLimitDegrees=45, twistMotion=Locked}
 {bone1Name=mast_cloth_A, bone2Name=dyn_cloth_A_01, swing1LimitDegrees=10, swing2LimitDegrees=10, twistLimitDegrees=45, twistMotion=Locked}
]
