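# Animation Blueprint for the BucketKick pickaxe, adding secondary physics motion to its cloth and chain bones.
# bRigidBodyEnabled selects RigidBody or per-bone AnimDynamics simulation, and bIsInFrontEnd picks each one's front-end variant, which adds fixed bone offsets to pose the parts.
# The event graph updates a gravity override through a deprecated helper, used only by the in-game RigidBody node.
# It also computes a Jump Alpha that the AnimGraph never reads.
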AnimBlueprint /BRCosmetics/Weapons/FORT_Melee/Pickaxe_BucketKick/Meshes/BucketKick_Axe_AnimBP
target_skeleton=&/BucketKick_Axe_Skeleton
graphs [
 AnimGraph {nodes [
   "Transform (Modify) Bone - Bone: mast_cloth_B"(Alpha=1, Translation="0,0,0", Rotation="25,0,25", Scale="1,1,1") {rotationMode=BMM_Additive, alphaBoolBlend {blendOption=Linear}, boneToModify {boneName=mast_cloth_B}, rotation {pitch=25, roll=25}}
   n1 = "Transform (Modify) Bone - Bone: mast_cloth_B"(
    ComponentPose=Local To Component(LocalPose=Input Pose() {name=InPose_1})
    Alpha=1
    Translation="0,0,0"
    Rotation="24.132,-6.747,8.868"
    Scale="1,1,1"
   ) {rotationMode=BMM_Additive, alphaBoolBlend {blendOption=Linear}, boneToModify {boneName=mast_cloth_B}, rotation=<rot 24.132 -6.747 8.868>}
   n2 = "Transform (Modify) Bone - Bone: mstr_chain"(ComponentPose=@n1, Alpha=1, Translation="0,0,0", Rotation="25,90,0", Scale="1,1,1") {rotationMode=BMM_Additive, alphaBoolBlend {blendOption=Linear}, boneToModify {boneName=mstr_chain}, rotation {pitch=25, yaw=90}}
   n3 = RigidBody!(ComponentPose=@n2, Alpha, OverrideWorldGravity:Vector="0,0,-980", ExternalForce) -> (Pose) {
    bOverrideWorldGravity=true
    overridePhysicsAsset=&/BucketKick_Axe_PhysicsAsset_FE
    alphaBoolBlend {blendOption=Linear}
    baseBoneRef {boneName=Melee_TwoHanded_Handle}
    componentAppliedLinearAccClamp=<3000 3000 3000>
    componentLinearAccScale=<0.4 0.4 0.4>
    componentLinearVelScale=<0.4 0.4 0.4>
   }
   n4 = "Transform (Modify) Bone - Bone: dyn_cloth_B_02"(ComponentPose=@n3, Alpha=1, Translation="0,0,0", Rotation="0.000,0.000,-11.132", Scale="1,1,1") {
    rotationMode=BMM_Additive
    translationMode=BMM_Additive
    alphaBoolBlend {blendOption=Linear}
    boneToModify {boneName=dyn_cloth_B_02}
    rotation {roll=-11.132}
   }
   n5 = "Transform (Modify) Bone - Bone: dyn_cloth_B_03"(ComponentPose=@n4, Alpha=1, Translation="0,0,0", Rotation="-0.000,0.000,36.717", Scale="1,1,1") {rotationMode=BMM_Additive, alphaBoolBlend {blendOption=Linear}, boneToModify {boneName=dyn_cloth_B_03}, rotation {roll=36.717}}
   n6 = Component To Local(ComponentPose=@n5)
   AnimGraphNode_SubInput_0 = Input Pose()
   n7 = RigidBody(
    ComponentPose=Local To Component(LocalPose=@AnimGraphNode_SubInput_0)
    OverrideWorldGravity=Get Graviti Override() -> (Graviti Override:Vector)
   ) {
    bOverrideWorldGravity=true
    overridePhysicsAsset=&/BucketKick_Axe_PhysicsAsset
    alphaBoolBlend {blendOption=Linear}
    baseBoneRef {boneName=Melee_TwoHanded_Handle}
    componentAppliedLinearAccClamp=<3000 3000 3000>
    componentLinearAccScale=<0.4 0.4 0.4>
    componentLinearVelScale=<0.4 0.4 0.4>
   }
   n8 = Component To Local(ComponentPose=@n7)
   n9 = Get bIsInFrontEnd()
   n10 = "Transform (Modify) Bone - Bone: mast_cloth_B"(ComponentPose=Local To Component(LocalPose=Input Pose() {name=InPose_2}), Alpha=1, Translation="0,0,0", Rotation="25,0,25", Scale="1,1,1") {rotationMode=BMM_Additive, alphaBoolBlend {blendOption=Linear}, boneToModify {boneName=mast_cloth_B}, rotation {pitch=25, roll=25}}
   n11 = Anim Dynamics - Bone: dyn_cloth_A_01!(ComponentPose=@n10, Alpha=1, GravityOverride:Vector="0,0,-980", ExternalForce="0,0,0") -> (Pose) {
    angularDampingOverride=0.98
    bOverrideAngularDamping=true
    bUseGravityOverride=true
    alphaBoolBlend {blendOption=Linear}
    boundBone {boneName=dyn_cloth_A_01}
    componentAppliedLinearAccClamp=<1000 1000 1000>
    componentLinearAccScale=<0.5 0.5 0.5>
    componentLinearVelScale=<0.5 0.5 0.5>
    gravityOverride {z=-980}
    physicsBodyDefinitions [{
      boundBone {boneName=dyn_cloth_A_01}
      boxExtents=<4 12 12>
      constraintSetup {angularLimitsMax=<5 5 5>, angularLimitsMin=<-5 -5 -5>, angularTarget=<0 0 0>}
      localJointOffset=<1 0 0>
     }]
   }
   n12 = Anim Dynamics - Bone: dyn_cloth_A_02!(ComponentPose=@n11, Alpha=1, GravityOverride:Vector="0,0,-980", ExternalForce="0,0,0") -> (Pose) {
    angularDampingOverride=0.98
    bOverrideAngularDamping=true
    bUseGravityOverride=true
    alphaBoolBlend {blendOption=Linear}
    boundBone {boneName=dyn_cloth_A_02}
    componentAppliedLinearAccClamp=<800 800 800>
    componentLinearAccScale=<0.5 0.5 0.5>
    componentLinearVelScale=<0.5 0.5 0.5>
    gravityOverride {z=-980}
    physicsBodyDefinitions [{
      boundBone {boneName=dyn_cloth_A_02}
      boxExtents=<4 10 10>
      constraintSetup {angularLimitsMax=<5 5 5>, angularLimitsMin=<-5 -5 -5>, angularTarget=<0 0 0>}
      localJointOffset=<1 0 0>
     }]
   }
   n13 = Anim Dynamics - Bone: dyn_cloth_A_03!(ComponentPose=@n12, Alpha=1, GravityOverride:Vector="0,0,-980", ExternalForce="0,0,0") -> (Pose) {
    angularDampingOverride=0.99
    bOverrideAngularDamping=true
    bUseGravityOverride=true
    bUseSphericalLimits=true
    alphaBoolBlend {blendOption=Linear}
    boundBone {boneName=dyn_cloth_A_03}
    componentAppliedLinearAccClamp=<650 650 650>
    componentLinearAccScale=<0.5 0.5 0.5>
    componentLinearVelScale=<0.5 0.5 0.5>
    gravityOverride {z=-980}
    physicsBodyDefinitions [{
      boundBone {boneName=dyn_cloth_A_03}
      boxExtents=<5 10 10>
      constraintSetup {angularLimitsMax=<5 5 5>, angularLimitsMin=<-5 -5 -5>, angularTarget=<0 0 0>}
      localJointOffset=<1 0 0>
     }]
    sphericalLimits [{limitRadius=5, drivingBone {boneName=mstr_chain}, sphereLocalOffset=<-2.0592 0 2.3>}]
   }
   n14 = Anim Dynamics - Bone: dyn_cloth_B_01!(ComponentPose=@n13, Alpha=1, GravityOverride:Vector="0,0,-500", ExternalForce="0,0,0") -> (Pose) {
    angularDampingOverride=0.98
    bOverrideAngularDamping=true
    bUseGravityOverride=true
    alphaBoolBlend {blendOption=Linear}
    boundBone {boneName=dyn_cloth_B_01}
    componentAppliedLinearAccClamp=<5000 5000 5000>
    componentLinearAccScale=<0.2 0.2 0.2>
    componentLinearVelScale=<0.3 0.3 0.4>
    gravityOverride {z=-500}
    physicsBodyDefinitions [{
      boundBone {boneName=dyn_cloth_B_01}
      boxExtents=<10 10 10>
      constraintSetup {angularLimitsMax=<0 25 0>, angularLimitsMin=<0 -25 0>, angularTarget=<0 0 0>}
      localJointOffset=<1 0 0>
     }]
   }
   n15 = Anim Dynamics - Bone: dyn_cloth_B_02!(ComponentPose=@n14, Alpha=1, GravityOverride:Vector="0,0,-500", ExternalForce="0,0,0") -> (Pose) {
    angularDampingOverride=0.98
    bOverrideAngularDamping=true
    bUseGravityOverride=true
    alphaBoolBlend {blendOption=Linear}
    boundBone {boneName=dyn_cloth_B_02}
    componentAppliedLinearAccClamp=<800 800 800>
    componentLinearAccScale=<0.2 0.2 0.2>
    componentLinearVelScale=<0.3 0.3 0.4>
    gravityOverride {z=-500}
    physicsBodyDefinitions [{
      boundBone {boneName=dyn_cloth_B_02}
      boxExtents=<10 10 10>
      constraintSetup {angularLimitsMax=<0 30 0>, angularLimitsMin=<0 -30 0>, angularTarget=<0 0 0>}
      localJointOffset=<1 0 0>
     }]
   }
   n16 = Anim Dynamics - Bone: dyn_cloth_B_03!(ComponentPose=@n15, Alpha=1, GravityOverride:Vector="0,0,-700", ExternalForce="0,0,0") -> (Pose) {
    angularDampingOverride=0.99
    bOverrideAngularDamping=true
    bUseGravityOverride=true
    bUseSphericalLimits=true
    alphaBoolBlend {blendOption=Linear}
    boundBone {boneName=dyn_cloth_B_03}
    componentAppliedLinearAccClamp=<3500 3500 3500>
    componentLinearAccScale=<0.2 0.2 0.2>
    componentLinearVelScale=<0.3 0.3 0.4>
    gravityOverride {z=-700}
    physicsBodyDefinitions [{
      boundBone {boneName=dyn_cloth_B_03}
      boxExtents=<10 10 10>
      constraintSetup {angularLimitsMax=<0 30 0>, angularLimitsMin=<0 -30 0>, angularTarget=<0 0 0>}
      localJointOffset=<1 0 0>
     }]
    sphericalLimits [{limitRadius=5, drivingBone {boneName=mstr_chain}, sphereLocalOffset=<-2.0592 0 2.3>}]
   }
   n17 = Anim Dynamics - Bone: dyn_chain_01!(ComponentPose=@n16, Alpha=1, GravityOverride:Vector="0,0,-980", ExternalForce="0,0,0") -> (Pose) {
    angularDampingOverride=0.98
    bOverrideAngularDamping=true
    bUseGravityOverride=true
    alphaBoolBlend {blendOption=Linear}
    boundBone {boneName=dyn_chain_01}
    componentAppliedLinearAccClamp=<1000 1000 1000>
    componentLinearAccScale=<0.5 0.5 0.5>
    componentLinearVelScale=<0.5 0.5 0.5>
    gravityOverride {z=-980}
    physicsBodyDefinitions [{
      boundBone {boneName=dyn_chain_01}
      boxExtents=<4 12 12>
      constraintSetup {angularLimitsMax=<15 30 30>, angularLimitsMin=<-15 -30 -30>, angularTarget=<0 0 0>}
      localJointOffset=<1 0 0>
     }]
   }
   n18 = Anim Dynamics - Bone: dyn_chain_02!(ComponentPose=@n17, Alpha=1, GravityOverride:Vector="0,0,-980", ExternalForce="0,0,0") -> (Pose) {
    angularDampingOverride=0.98
    bOverrideAngularDamping=true
    bUseGravityOverride=true
    alphaBoolBlend {blendOption=Linear}
    boundBone {boneName=dyn_chain_02}
    componentAppliedLinearAccClamp=<800 800 800>
    componentLinearAccScale=<0.5 0.5 0.5>
    componentLinearVelScale=<0.5 0.5 0.5>
    gravityOverride {z=-980}
    physicsBodyDefinitions [{
      boundBone {boneName=dyn_chain_02}
      boxExtents=<4 10 10>
      constraintSetup {angularLimitsMax=<15 25 25>, angularLimitsMin=<-15 -25 -25>, angularTarget=<0 0 0>}
      localJointOffset=<1 0 0>
     }]
   }
   n19 = Anim Dynamics - Bone: dyn_chain_03!(ComponentPose=@n18, Alpha=1, GravityOverride:Vector="0,0,-980", ExternalForce="0,0,0") -> (Pose) {
    angularDampingOverride=0.99
    bOverrideAngularDamping=true
    bUseGravityOverride=true
    bUsePlanarLimit=false
    alphaBoolBlend {blendOption=Linear}
    boundBone {boneName=dyn_chain_03}
    componentAppliedLinearAccClamp=<650 650 650>
    componentLinearAccScale=<0.5 0.5 0.5>
    componentLinearVelScale=<0.5 0.5 0.5>
    gravityOverride {z=-980}
    physicsBodyDefinitions [{
      boundBone {boneName=dyn_chain_03}
      boxExtents=<5 10 10>
      constraintSetup {angularLimitsMax=<15 25 25>, angularLimitsMin=<-15 -25 -25>, angularTarget=<0 0 0>}
      localJointOffset=<1 0 0>
     }]
   }
   n20 = Anim Dynamics - Bone: dyn_chain_04!(ComponentPose=@n19, Alpha=1, GravityOverride:Vector="0,0,-980", ExternalForce="0,0,0") -> (Pose) {
    angularDampingOverride=0.99
    bOverrideAngularDamping=true
    bUseGravityOverride=true
    bUsePlanarLimit=false
    bUseSphericalLimits=true
    alphaBoolBlend {blendOption=Linear}
    boundBone {boneName=dyn_chain_04}
    componentAppliedLinearAccClamp=<650 650 650>
    componentLinearAccScale=<0.5 0.5 0.5>
    componentLinearVelScale=<0.5 0.5 0.5>
    gravityOverride {z=-980}
    physicsBodyDefinitions [{
      boundBone {boneName=dyn_chain_04}
      boxExtents=<5 10 10>
      constraintSetup {angularLimitsMax=<15 25 25>, angularLimitsMin=<-15 -25 -25>, angularTarget=<0 0 0>}
      localJointOffset=<1 0 0>
     }]
   }
   n21 = Component To Local(ComponentPose=@n20)
   n22 = Anim Dynamics - Bone: dyn_cloth_A_01(ComponentPose=Local To Component(LocalPose=Input Pose() {name=InPose_0}), Alpha=0.3, GravityOverride="0,0,-980", ExternalForce="0,0,0") {
    angularDampingOverride=0.98
    bOverrideAngularDamping=true
    bUseGravityOverride=true
    simulationSpace=World
    alphaBoolBlend {blendOption=Linear}
    boundBone {boneName=dyn_cloth_A_01}
    componentAppliedLinearAccClamp=<1000 1000 1000>
    componentLinearAccScale=<0.5 0.5 0.5>
    componentLinearVelScale=<0.5 0.5 0.5>
    gravityOverride {z=-980}
    physicsBodyDefinitions [{
      boundBone {boneName=dyn_cloth_A_01}
      boxExtents=<4 12 12>
      constraintSetup {angularLimitsMax=<5 5 5>, angularLimitsMin=<-5 -5 -5>, angularTarget=<0 0 0>}
      localJointOffset=<1 0 0>
     }]
   }
   n23 = Anim Dynamics - Bone: dyn_cloth_A_02(ComponentPose=@n22, Alpha=0.3, GravityOverride="0,0,-980", ExternalForce="0,0,0") {
    angularDampingOverride=0.98
    bOverrideAngularDamping=true
    bUseGravityOverride=true
    simulationSpace=World
    alphaBoolBlend {blendOption=Linear}
    boundBone {boneName=dyn_cloth_A_02}
    componentAppliedLinearAccClamp=<800 800 800>
    componentLinearAccScale=<0.5 0.5 0.5>
    componentLinearVelScale=<0.5 0.5 0.5>
    gravityOverride {z=-980}
    physicsBodyDefinitions [{
      boundBone {boneName=dyn_cloth_A_02}
      boxExtents=<4 10 10>
      constraintSetup {angularLimitsMax=<5 5 5>, angularLimitsMin=<-5 -5 -5>, angularTarget=<0 0 0>}
      localJointOffset=<1 0 0>
     }]
   }
   n24 = Anim Dynamics - Bone: dyn_cloth_A_03(ComponentPose=@n23, Alpha=0.2, GravityOverride="0,0,-980", ExternalForce="0,0,0") {
    angularDampingOverride=0.99
    bOverrideAngularDamping=true
    bUseGravityOverride=true
    bUseSphericalLimits=true
    simulationSpace=World
    alphaBoolBlend {blendOption=Linear}
    boundBone {boneName=dyn_cloth_A_03}
    componentAppliedLinearAccClamp=<650 650 650>
    componentLinearAccScale=<0.5 0.5 0.5>
    componentLinearVelScale=<0.5 0.5 0.5>
    gravityOverride {z=-980}
    physicsBodyDefinitions [{
      boundBone {boneName=dyn_cloth_A_03}
      boxExtents=<5 10 10>
      constraintSetup {angularLimitsMax=<5 5 5>, angularLimitsMin=<-5 -5 -5>, angularTarget=<0 0 0>}
      localJointOffset=<1 0 0>
     }]
    sphericalLimits [{limitRadius=5, drivingBone {boneName=mstr_chain}, sphereLocalOffset=<-2.0592 0 2.3>}]
   }
   n25 = Anim Dynamics - Bone: dyn_cloth_B_01(ComponentPose=@n24, Alpha=0.3, GravityOverride="0,0,-500", ExternalForce="0,0,0") {
    angularDampingOverride=0.98
    bOverrideAngularDamping=true
    bUseGravityOverride=true
    simulationSpace=World
    alphaBoolBlend {blendOption=Linear}
    boundBone {boneName=dyn_cloth_B_01}
    componentAppliedLinearAccClamp=<5000 5000 5000>
    componentLinearAccScale=<0.2 0.2 0.2>
    componentLinearVelScale=<0.3 0.3 0.4>
    gravityOverride {z=-500}
    physicsBodyDefinitions [{
      boundBone {boneName=dyn_cloth_B_01}
      boxExtents=<10 10 10>
      constraintSetup {angularLimitsMax=<0 5 0>, angularLimitsMin=<0 -5 0>, angularTarget=<0 0 0>}
      localJointOffset=<1 0 0>
     }]
   }
   n26 = Anim Dynamics - Bone: dyn_cloth_B_02(ComponentPose=@n25, Alpha=0.3, GravityOverride="0,0,-500", ExternalForce="0,0,0") {
    angularDampingOverride=0.98
    bOverrideAngularDamping=true
    bUseGravityOverride=true
    simulationSpace=World
    alphaBoolBlend {blendOption=Linear}
    boundBone {boneName=dyn_cloth_B_02}
    componentAppliedLinearAccClamp=<800 800 800>
    componentLinearAccScale=<0.2 0.2 0.2>
    componentLinearVelScale=<0.3 0.3 0.4>
    gravityOverride {z=-500}
    physicsBodyDefinitions [{
      boundBone {boneName=dyn_cloth_B_02}
      boxExtents=<10 10 10>
      constraintSetup {angularLimitsMax=<0 5 0>, angularLimitsMin=<0 -5 0>, angularTarget=<0 0 0>}
      localJointOffset=<1 0 0>
     }]
   }
   n27 = Anim Dynamics - Bone: dyn_cloth_B_03(ComponentPose=@n26, Alpha=0.2, GravityOverride="0,0,-700", ExternalForce="0,0,0") {
    angularDampingOverride=0.99
    bOverrideAngularDamping=true
    bUseGravityOverride=true
    bUseSphericalLimits=true
    simulationSpace=World
    alphaBoolBlend {blendOption=Linear}
    boundBone {boneName=dyn_cloth_B_03}
    componentAppliedLinearAccClamp=<3500 3500 3500>
    componentLinearAccScale=<0.2 0.2 0.2>
    componentLinearVelScale=<0.3 0.3 0.4>
    gravityOverride {z=-700}
    physicsBodyDefinitions [{
      boundBone {boneName=dyn_cloth_B_03}
      boxExtents=<10 10 10>
      constraintSetup {angularLimitsMax=<0 5 0>, angularLimitsMin=<0 -5 0>, angularTarget=<0 0 0>}
      localJointOffset=<1 0 0>
     }]
    sphericalLimits [{limitRadius=5, drivingBone {boneName=mstr_chain}, sphereLocalOffset=<-2.0592 0 2.3>}]
   }
   n28 = "Transform (Modify) Bone - Bone: dyn_chain_01"(ComponentPose=@n27, Alpha=1, Translation="0,0,0", Rotation="-19.471,26.307,-6.699", Scale="1,1,1") {rotationMode=BMM_Additive, alphaBoolBlend {blendOption=Linear}, boneToModify {boneName=dyn_chain_01}, rotation=<rot -19.471 26.307 -6.699>}
   n29 = "Transform (Modify) Bone - Bone: dyn_chain_02"(ComponentPose=@n28, Alpha=1, Translation="0,0,0", Rotation="6.718,6.138,-22.539", Scale="1,1,1") {rotationMode=BMM_Additive, alphaBoolBlend {blendOption=Linear}, boneToModify {boneName=dyn_chain_02}, rotation=<rot 6.718 6.138 -22.539>}
   n30 = "Transform (Modify) Bone - Bone: dyn_chain_03"(ComponentPose=@n29, Alpha=1, Translation="0,0,0", Rotation="31.676,178.487,43.551", Scale="1,1,1") {rotationMode=BMM_Additive, alphaBoolBlend {blendOption=Linear}, boneToModify {boneName=dyn_chain_03}, rotation=<rot 31.676 178.487 43.551>}
   n31 = "Transform (Modify) Bone - Bone: dyn_chain_04"(ComponentPose=@n30, Alpha=1, Translation="0,0,0", Rotation="12.335,-1.673,-2.278", Scale="1,1,1") {rotationMode=BMM_Additive, alphaBoolBlend {blendOption=Linear}, boneToModify {boneName=dyn_chain_04}, rotation=<rot 12.335 -1.673 -2.278>}
   n32 = Anim Dynamics - Bone: dyn_chain_01(ComponentPose=@n31, Alpha=0.2, GravityOverride="0,0,-980", ExternalForce="0,0,0") {
    angularDampingOverride=0.98
    bOverrideAngularDamping=true
    bUseGravityOverride=true
    simulationSpace=World
    alphaBoolBlend {blendOption=Linear}
    boundBone {boneName=dyn_chain_01}
    componentAppliedLinearAccClamp=<500 500 500>
    componentLinearAccScale=<0.1 0.1 0.1>
    componentLinearVelScale=<0.2 0.2 0.2>
    gravityOverride {z=-980}
    physicsBodyDefinitions [{
      boundBone {boneName=dyn_chain_01}
      boxExtents=<4 12 12>
      constraintSetup {angularLimitsMax=<0 0 0>, angularLimitsMin=<0 0 0>, angularTarget=<0 0 0>}
      localJointOffset=<1 0 0>
     }]
   }
   n33 = Anim Dynamics - Bone: dyn_chain_02(ComponentPose=@n32, Alpha=0.2, GravityOverride="0,0,-980", ExternalForce="0,0,0") {
    angularDampingOverride=0.98
    bOverrideAngularDamping=true
    bUseGravityOverride=true
    simulationSpace=World
    alphaBoolBlend {blendOption=Linear}
    boundBone {boneName=dyn_chain_02}
    componentAppliedLinearAccClamp=<500 500 500>
    componentLinearAccScale=<0.1 0.1 0.1>
    componentLinearVelScale=<0.2 0.2 0.2>
    gravityOverride {z=-980}
    physicsBodyDefinitions [{
      boundBone {boneName=dyn_chain_02}
      boxExtents=<4 10 10>
      constraintSetup {angularLimitsMax=<0 0 0>, angularLimitsMin=<0 0 0>, angularTarget=<0 0 0>}
      localJointOffset=<1 0 0>
     }]
   }
   n34 = Anim Dynamics - Bone: dyn_chain_03(ComponentPose=@n33, Alpha=0.2, GravityOverride="0,0,-980", ExternalForce="0,0,0") {
    angularDampingOverride=0.99
    bOverrideAngularDamping=true
    bUseGravityOverride=true
    bUsePlanarLimit=false
    simulationSpace=World
    alphaBoolBlend {blendOption=Linear}
    boundBone {boneName=dyn_chain_03}
    componentAppliedLinearAccClamp=<500 500 500>
    componentLinearAccScale=<0.1 0.1 0.1>
    componentLinearVelScale=<0.2 0.2 0.2>
    gravityOverride {z=-980}
    physicsBodyDefinitions [{
      boundBone {boneName=dyn_chain_03}
      boxExtents=<5 10 10>
      constraintSetup {angularLimitsMax=<0 0 0>, angularLimitsMin=<0 0 0>, angularTarget=<0 0 0>}
      localJointOffset=<1 0 0>
     }]
   }
   n35 = Anim Dynamics - Bone: dyn_chain_04(ComponentPose=@n34, Alpha=0.2, GravityOverride="0,0,-980", ExternalForce="0,0,0") {
    angularDampingOverride=0.99
    bOverrideAngularDamping=true
    bUseGravityOverride=true
    bUsePlanarLimit=false
    bUseSphericalLimits=true
    simulationSpace=World
    alphaBoolBlend {blendOption=Linear}
    boundBone {boneName=dyn_chain_04}
    componentAppliedLinearAccClamp=<500 500 500>
    componentLinearAccScale=<0.1 0.1 0.1>
    componentLinearVelScale=<0.2 0.2 0.2>
    gravityOverride {z=-980}
    physicsBodyDefinitions [{
      boundBone {boneName=dyn_chain_04}
      boxExtents=<5 10 10>
      constraintSetup {angularLimitsMax=<0 0 0>, angularLimitsMin=<0 0 0>, angularTarget=<0 0 0>}
      localJointOffset=<1 0 0>
     }]
   }
   n36 = "Transform (Modify) Bone - Bone: mast_cloth_B"(ComponentPose=@n35, Alpha=1, Translation="0,0,0", Rotation="24.132,-6.747,8.868", Scale="1,1,1") {rotationMode=BMM_Additive, alphaBoolBlend {blendOption=Linear}, boneToModify {boneName=mast_cloth_B}, rotation=<rot 24.132 -6.747 8.868>}
   n37 = "Transform (Modify) Bone - Bone: dyn_cloth_B_01"(ComponentPose=@n36, Alpha=1, Translation="0,1,0", Rotation="-16.167,-0.000,-3.817", Scale="1,1,1") {
    rotationMode=BMM_Additive
    translationMode=BMM_Additive
    alphaBoolBlend {blendOption=Linear}
    boneToModify {boneName=dyn_cloth_B_01}
    rotation {pitch=-16.167, roll=-3.817}
    translation {y=1}
   }
   n38 = "Transform (Modify) Bone - Bone: dyn_cloth_B_02"(ComponentPose=@n37, Alpha=1, Translation="0,0,0", Rotation="0.000,0.000,-27.213", Scale="1,1,1") {
    rotationMode=BMM_Additive
    translationMode=BMM_Additive
    alphaBoolBlend {blendOption=Linear}
    boneToModify {boneName=dyn_cloth_B_02}
    rotation {roll=-27.213}
   }
   n39 = "Transform (Modify) Bone - Bone: dyn_cloth_B_03"(ComponentPose=@n38, Alpha=1, Translation="0,0,0", Rotation="-3.757,-4.465,11.744", Scale="1,1,1") {rotationMode=BMM_Additive, alphaBoolBlend {blendOption=Linear}, boneToModify {boneName=dyn_cloth_B_03}, rotation=<rot -3.757 -4.465 11.744>}
   n40 = "Transform (Modify) Bone - Bone: dyn_cloth_A_02"(ComponentPose=@n39, Alpha=1, Translation="0,1,0", Rotation="0.000,0.000,-25.887", Scale="1,1,1") {
    rotationMode=BMM_Additive
    translationMode=BMM_Additive
    alphaBoolBlend {blendOption=Linear}
    boneToModify {boneName=dyn_cloth_A_02}
    rotation {roll=-25.887}
    translation {y=1}
   }
   n41 = "Transform (Modify) Bone - Bone: dyn_cloth_A_03"(ComponentPose=@n40, Alpha=1, Translation="0,0,0", Rotation="-0.000,0.000,1.537", Scale="1,1,1") {rotationMode=BMM_Additive, alphaBoolBlend {blendOption=Linear}, boneToModify {boneName=dyn_cloth_A_03}, rotation {roll=1.537}}
   n42 = Component To Local(ComponentPose=@n41)
   n43 = Get bIsInFrontEnd()
   Output Pose(Result=Blend Poses by bool(
     BlendPose_0=Blend Poses by bool(BlendPose_0=@n6, BlendPose_1=@n8, bActiveValue=@n9) {blendType=Linear}
     BlendPose_1=Blend Poses by bool(BlendPose_0=@n42, BlendPose_1=@n21, bActiveValue=@n43) {blendType=Linear}
     bActiveValue=Get bRigidBodyEnabled()
    ) {blendType=Linear}) {layerGroup=None}
  ]}
 EventGraph {nodes [
   n1 = Event BlueprintUpdateAnimation()
   > n2 = Sequence!(execute) -> (then_0, then_1, then_2:Exec, then_3:Exec, then_4:Exec)
   > n3 = Gravity Override State Function DEPRECATED(CharacterPartAnimInstance=Self-Reference(), GameTimeInSeconds=GetGameTimeInSeconds(), GravityOverrideParams=Get Gravity Override Params())
   > Set Graviti Override(Graviti Override:Vector=@n3) -> (Output_Get:Vector)
   Set interpSpeed(execute=@n2.then_1, interpSpeed:float=0) -> (Output_Get:float)
   n4 = Branch(execute=@n2.then_2, Condition=OR Boolean(A=Get bIsJumping(), B=Get bIsFalling()))
   > Set interpSpeed(interpSpeed=8)
   Set interpSpeed(execute=@n4.else, interpSpeed=0)
   n5 = Get interpSpeed() -> (interpSpeed:float)
   Set Jump Alpha(execute=@n2.then_3, Jump Alpha:double=Lerp(A=0, B=1, Alpha=FInterpTo(Current=1, Target=0, DeltaTime=@n1.DeltaTimeX, InterpSpeed=@n5))) -> (Output_Get:double)
   Event BlueprintInitializeAnimation()
   > Set Skel Mesh(Skel Mesh=GetPartSkeletalMeshComponent())
   "To String (Rotator)"!(InRot:Rotator="0, 0, 0") -> (ReturnValue:String)
   TryGetPawnOwner()
  ]}
]
variables [
 Alpha Zipper 01 L {type=number}
 Graviti Override {type=Vector}
 Graviti Override Jump {type=Vector}
 Gravity Override Params {type=GravityOverrideParamsStruct}
 interpSpeed {type=number}
 Is Jumping and Falling {type=boolean}
 Jump Alpha {type=number}
 Skel Mesh {type=SkeletalMeshComponent}
]
